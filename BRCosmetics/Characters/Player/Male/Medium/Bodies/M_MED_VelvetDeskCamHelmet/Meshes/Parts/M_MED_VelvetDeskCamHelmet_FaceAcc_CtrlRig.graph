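# Control Rig for the VelvetDeskCamHelmet face accessory that applies pose-driven offsets to hair and hat bones.
# From Begin Execution, a spherical pose reader on head relative to spine_04 weights a base_hair offset, then isSkydiving weights a second.
# A separate hat chain, not linked to Begin Execution, multiplies two swing-based Delta Rotation weights from neck and spine, curve-shaping the result to offset hat bones.
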ControlRigBlueprint /BRCosmetics/Characters/Player/Male/Medium/Bodies/M_MED_VelvetDeskCamHelmet/Meshes/Parts/M_MED_VelvetDeskCamHelmet_FaceAcc_CtrlRig
graphs [
 New Function_ContainedGraph {nodes [
   n1 = RigVMFunctionEntryNode()
   Clamp = &RigVMFunction_MathIntClamp(Value=@n1.primAxis)
   RigVMFunction_MathBoolToInteger = &RigVMFunction_MathBoolToInteger(Value=@n1.onlyTwist)
   RigVMFunction_MathQuaternionSwingTwist = &RigVMFunction_MathQuaternionSwingTwist(
    TwistAxis=@n1.twistAxis
    Input=&RigUnit_GetRelativeTransformForItem(
     Child=(Type=Bone,Name="None")
     Parent=(Type=Bone,Name="None")
     bChildInitial=False
     bParentInitial=False
     Parent.Type=@n1.ParentType
     Child.Name=@n1.child
     Parent.Name=@n1.parent
    ).RelativeTransform.Rotation
   )
   RigVMFunction_MathQuaternionToEuler = &RigVMFunction_MathQuaternionToEuler(
    Value=RigVMDispatchNode(
     Values=((),())
     Index=@RigVMFunction_MathBoolToInteger.Result
     Values.0=@RigVMFunction_MathQuaternionSwingTwist.Swing
     Values.1=@RigVMFunction_MathQuaternionSwingTwist.Twist
    ) {template="DISPATCH_RigVMDispatch_SelectInt32(in Index,in Values,out Result)"}.Result
    RotationOrder=@n1.rotationOrder
   )
   DISPATCH_RigVMDispatch_SelectInt32 = RigVMDispatchNode(
    Values=(0,0,0)
    Index=@Clamp.Result
    Values.0=@RigVMFunction_MathQuaternionToEuler.Result.X
    Values.1=@RigVMFunction_MathQuaternionToEuler.Result.Y
    Values.2=@RigVMFunction_MathQuaternionToEuler.Result.Z
   ) {template="DISPATCH_RigVMDispatch_SelectInt32(in Index,in Values,out Result)"}
   Remap = &RigVMFunction_MathDoubleRemap(
    TargetMaximum=1
    TargetMinimum=0
    bClamp=False
    Value=@DISPATCH_RigVMDispatch_SelectInt32.Result
    SourceMaximum=@n1.max
    SourceMinimum=@n1.min
   )
   > RigVMFunctionReturnNode(ouputWeight=&RigVMFunction_MathDoubleClamp(Value=@Remap.Result).Result)
  ]}
 RigVMFunctionLibrary {nodes [Delta Rotation = RigVMCollapseNode(ExecuteContext=(), debugRotation=False, onlyTwist=False, twistAxis=(X=0,Y=0,Z=0))]}
 RigVMModel {nodes [
   Delta Rotation = RigVMFunctionReferenceNode(
    ExecuteContext=()
    ParentType=Bone
    child=neck_02
    debugRotation=False
    max=45
    min=10
    onlyTwist=False
    parent=head
    primAxis=2
    rotationOrder=ZXY
    twistAxis=(X=1,Y=0,Z=0)
   )
   > Delta Rotation_1 = RigVMFunctionReferenceNode(
    ParentType=Bone
    child=spine_05
    debugRotation=False
    max=23
    min=15
    onlyTwist=False
    parent=neck_01
    primAxis=1
    rotationOrder=ZXY
    twistAxis=(X=1,Y=0,Z=0)
   )
   Multiply = &RigVMFunction_MathDoubleMul(B=@"Delta Rotation".ouputWeight, A=@"Delta Rotation_1".ouputWeight)
   RigVMFunction_AnimEvalRichCurve = &RigVMFunction_AnimEvalRichCurve(
    Curve=(EditorCurveData=(Keys=((),(InterpMode=RCIM_Cubic,TangentMode=RCTM_User,Time=0.2499,Value=0.0198,ArriveTangent=0.1589,LeaveTangent=0.1589),(Time=1,Value=1))))
    SourceMinimum=0
    TargetMinimum=0
    Value=@Multiply.Result
   )
   > &RigUnit_ModifyTransforms(
    ItemToModify="((Item=(Type=Bone,Name=\"base_hat_area\"),Transform=(Rotation=(X=0,Y=-0.0872,Z=0,W=0.9962),Translation=(X=-0.6673,Y=0.8653,Z=0.7477),Scale3D=(X=1,Y=1,Z=1))),(Item=(Type=Bone,Name=\"base_hat\"),Transform=(Rotation=(X=-0.0435,Y=-0.0871,Z=-0.0038,W=0.9952),Transl ...(truncated)"
    WeightMinimum=0
    Weight=@RigVMFunction_AnimEvalRichCurve.Result
   )
   RigUnit_BeginExecution = &RigUnit_BeginExecution()
   n1 = &RigUnit_SphericalPoseReader(
    ActiveRegionScaleFactors=(PositiveWidth=3,NegativeWidth=2,PositiveHeight=6,NegativeHeight=6)
    ActiveRegionSize=0.1
    Debug=(bDrawDebug=True,DebugScale=25,DebugSegments=20,DebugThickness=0.25)
    DriverAxis=(X=1,Y=0,Z=0)
    DriverItem=(Type=Bone,Name="head")
    FalloffRegionScaleFactors=(PositiveWidth=-3,NegativeWidth=2,PositiveHeight=0,NegativeHeight=0)
    FalloffSize=0.25
    OptionalParentItem=(Type=Bone,Name="spine_04")
    RotationOffset=(X=90,Y=0,Z=-180)
    ExecutePin=@RigUnit_BeginExecution.ExecutePin
   )
   > &RigUnit_ModifyTransforms(
    ItemToModify=((Item=(Type=Bone,Name="base_hair"),Transform=(Rotation=(X=0,Y=-0.5373,Z=0,W=0.8434),Translation=(X=-1,Y=0,Z=0),Scale3D=(X=1,Y=1,Z=1))))
    Weight=@n1.OutputParam
   )
   > &RigUnit_ModifyTransforms(
    ItemToModify=((Item=(Type=Bone,Name="base_hair"),Transform=(Rotation=(X=0,Y=0.0872,Z=0,W=0.9962),Translation=(X=-1,Y=0,Z=-1),Scale3D=(X=1,Y=1,Z=1))))
    Weight=RigVMVariableNode() {variable=isSkydiving}.Value
   )
  ]}
]
hierarchy [
 root {loc=<0 0 0>, rot=<0 0 0 1>}
 pelvis {parent=root, loc=<0 -0.7071 93.1191>, rot=<0.0127 0.707 -0.0127 -0.707>}
 spine_01 {parent=pelvis, loc=<6.7494 -0.6388 0>, rot=<0 0 0.034 -0.9994>}
 spine_02 {parent=spine_01, loc=<7.252 -0.035 0>, rot=<0 0 0.0238 -0.9997>}
 spine_03 {parent=spine_02, loc=<6.0106 0 0>, rot=<0 0 0.0149 -0.9999>}
 spine_04 {parent=spine_03, loc=<8.1023 0 0>, rot=<0 0 -0.0132 -0.9999>}
 spine_05 {parent=spine_04, loc=<10.0142 0 0>, rot=<0 0 -0.0015 -1>}
 neck_01 {parent=spine_05, loc=<11.3138 0.1529 0>, rot=<0 0 -0.2003 -0.9797>}
 neck_02 {parent=neck_01, loc=<4.5627 0 0>, rot=<0 0 0.0504 -0.9987>}
 head {parent=neck_02, loc=<5.3914 0 0>, rot=<0 0 0.115 -0.9934>}
 base_hair_sd_l_A {parent=head, loc=<9.4611 8.0564 -4.1544>, rot=<0.0245 -0.9223 -0.36 -0.138>}
 dyn_hair_sd_l_A_01 {parent=base_hair_sd_l_A, loc=<0 0 0>, rot=<0 0 0 1>}
 dyn_hair_sd_l_A_02 {parent=dyn_hair_sd_l_A_01, loc=<2.9791 0 0>, rot=<0 0 0 1>}
 base_hair_sd_l_B {parent=head, loc=<7.9845 6.9753 -5.5874>, rot=<0.1833 -0.8893 0.2832 -0.3087>}
 dyn_hair_sd_l_B_01 {parent=base_hair_sd_l_B, loc=<0 0 0>, rot=<0 0 0 1>}
 dyn_hair_sd_l_B_02 {parent=dyn_hair_sd_l_B_01, loc=<2.7443 0 0>, rot=<0 0 0 1>}
 base_hair_sd_l_C {parent=head, loc=<6.1888 4.2618 -6.6149>, rot=<0.2579 -0.7626 0.3809 -0.4547>}
 dyn_hair_sd_l_C_01 {parent=base_hair_sd_l_C, loc=<0 0 0>, rot=<0 0 0 1>}
 dyn_hair_sd_l_C_02 {parent=dyn_hair_sd_l_C_01, loc=<2.7443 0 0>, rot=<0 0 0 1>}
 base_hair_fw_area {parent=head, loc=<12.591 9.0729 0.8106>, rot=<0.4979 0.5021 0.5021 0.4979>}
 dyn_hair_fw_area {parent=base_hair_fw_area, loc=<0 0 0>, rot=<0 0 0 1>}
 base_fw_C {parent=dyn_hair_fw_area, loc=<0.0948 3.6478 -1.4544>, rot=<0.0475 -0.1297 -0.7534 -0.6428>}
 dyn_hair_fw_C_01 {parent=base_fw_C, loc=<0 0 0>, rot=<0 0 0 1>}
 dyn_hair_fw_C_02 {parent=dyn_hair_fw_C_01, loc=<2.0457 0 0>, rot=<0 0 0 1>}
 base_hair_fw_B {parent=dyn_hair_fw_area, loc=<1.4776 3.7495 -3.48>, rot=<0.0581 -0.2771 -0.6506 -0.7047>}
 dyn_hair_fw_B_01 {parent=base_hair_fw_B, loc=<0 0 0>, rot=<0 0 0 1>}
 dyn_hair_fw_B_02 {parent=dyn_hair_fw_B_01, loc=<2 0 0>, rot=<0 0.0032 0 -1>}
 dyn_hair_fw_B_03 {parent=dyn_hair_fw_B_01, loc=<1.6786 0 -1.3506>, rot=<-0.1945 -0.4119 0.0904 -0.8856>}
 base_fw_A {parent=dyn_hair_fw_area, loc=<1.188 -0.8009 -1.0345>, rot=<-0.0661 -0.2992 -0.2981 -0.904>}
 dyn_hair_fw_A_01 {parent=base_fw_A, loc=<0 0 0.5165>, rot=<0 -0.2327 0 -0.9726>}
 dyn_hair_fw_A_02 {parent=dyn_hair_fw_A_01, loc=<2 0 0>, rot=<0 -0.0069 0 -1>}
 dyn_hair_fw_A_03 {parent=dyn_hair_fw_A_02, loc=<2 0 0>, rot=<0 0.0005 0 -1>}
 base_hair_sd_r_A {parent=head, loc=<5.5029 5.324 6.6909>, rot=<0.4723 -0.4909 0.6277 -0.3768>}
 dyn_hair_sd_r_A_01 {parent=base_hair_sd_r_A, loc=<0 0 0>, rot=<0 0 0 1>}
 dyn_hair_sd_r_A_02 {parent=dyn_hair_sd_r_A_01, loc=<3.3117 0 0>, rot=<0 0 0 1>}
 base_hat_area {parent=head, loc=<10.1374 1.2385 0>, rot=<0.5021 -0.4979 0.4979 -0.5021>}
 dyn_hat_area {parent=base_hat_area, loc=<0 0 0>, rot=<0 0 0 1>}
 base_hat {parent=dyn_hat_area, loc=<7.533 0 -3.1348>, rot=<0 0 0 1>}
 dyn_hat {parent=base_hat, loc=<0 0 0>, rot=<0 0 0 1>}
 base_hair {parent=head, loc=<0.1814 -3.9332 0>, rot=<0.5021 -0.4979 0.4979 -0.5021>}
 dyn_hair {parent=base_hair, loc=<0 0 0>, rot=<0 0 0 1>}
 clavicle_l {parent=spine_05, loc=<4.3575 -0.1466 -2.3223>, rot=<0.0465 -0.7289 -0.009 -0.6829>}
 clavicle_r {parent=spine_05, loc=<4.3575 -0.1466 2.3223>, rot=<0.7289 0.0465 -0.6829 0.009>}
]
variables [isSkydiving {type=double, default=0}]
