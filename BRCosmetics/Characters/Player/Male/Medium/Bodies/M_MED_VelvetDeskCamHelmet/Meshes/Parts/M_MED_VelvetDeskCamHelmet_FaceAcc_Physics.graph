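# Physics asset for the face accessory mesh of the VelvetDeskCamHelmet male medium body cosmetic, covering its dyn_hair_fw hair bones.
# Capsule bodies form three groups (A, B, C) rooted on a head body, which has two capsules.
# A and C are linear chains, while B branches, with both B_02 and B_03 constrained to B_01.
# All constraints lock twist and use tight swing limits, loosest at the head-to-B joint.
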PhysicsAsset /BRCosmetics/Characters/Player/Male/Medium/Bodies/M_MED_VelvetDeskCamHelmet/Meshes/Parts/M_MED_VelvetDeskCamHelmet_FaceAcc_Physics
bodies {
 dyn_hair_fw_A_01 [{shapeType=Capsule, length=1, radius=1, shapeName=dyn_hair_fw_A_01_capsule, rotation=<rot 90 0 0>}]
 dyn_hair_fw_A_02 [{shapeType=Capsule, length=1, radius=1, shapeName=dyn_hair_fw_A_02_capsule, rotation=<rot 90 0 0>}]
 dyn_hair_fw_A_03 [{shapeType=Capsule, length=2.5, radius=1.5, shapeName=dyn_hair_fw_A_03_capsule, center=<1.742 -0.3833 -0.7153>, rotation=<rot 58.3651 0 0>}]
 dyn_hair_fw_B_01 [{
   shapeType=Capsule
   length=1.9301
   radius=1.0938
   shapeName=dyn_hair_fw_B_01_capsule
   center=<0.9117 -0.2382 -0.2476>
   rotation=<rot 0 -70.4009 -65.5779>
  }]
 dyn_hair_fw_B_02 [{shapeType=Capsule, length=1, radius=1, shapeName=dyn_hair_fw_B_02_capsule, rotation=<rot 90 0 0>}]
 dyn_hair_fw_B_03 [{shapeType=Capsule, length=1, radius=1, shapeName=dyn_hair_fw_B_03_capsule, rotation=<rot 90 0 0>}]
 dyn_hair_fw_C_01 [{shapeType=Capsule, length=1, radius=1, shapeName=dyn_hair_fw_C_01_capsule, rotation=<rot 90 0 0>}]
 dyn_hair_fw_C_02 [{shapeType=Capsule, length=1, radius=1, shapeName=dyn_hair_fw_C_02_capsule, rotation=<rot 90 0 0>}]
 head [
  {shapeType=Capsule, length=8, radius=10, shapeName=head_capsule, center=<5.795 -0.1243 0>}
  {shapeType=Capsule, length=4.5, radius=2, shapeName=head_capsule, center=<9.5739 12.1399 5.1876>, rotation=<rot 57.9384 -0.0001 0>}
 ]
}
constraints [
 {
  bone1Name=dyn_hair_fw_A_03
  bone2Name=dyn_hair_fw_A_02
  swing1LimitDegrees=10
  swing2LimitDegrees=10
  twistLimitDegrees=45
  twistMotion=Locked
 }
 {
  bone1Name=dyn_hair_fw_A_02
  bone2Name=dyn_hair_fw_A_01
  swing1LimitDegrees=10
  swing2LimitDegrees=10
  twistLimitDegrees=45
  twistMotion=Locked
 }
 {
  bone1Name=dyn_hair_fw_C_02
  bone2Name=dyn_hair_fw_C_01
  swing1LimitDegrees=10
  swing2LimitDegrees=10
  twistLimitDegrees=45
  twistMotion=Locked
 }
 {
  bone1Name=dyn_hair_fw_B_02
  bone2Name=dyn_hair_fw_B_01
  swing1LimitDegrees=10
  swing2LimitDegrees=10
  twistLimitDegrees=45
  twistMotion=Locked
 }
 {
  bone1Name=dyn_hair_fw_B_03
  bone2Name=dyn_hair_fw_B_01
  swing1LimitDegrees=10
  swing2LimitDegrees=10
  twistLimitDegrees=45
  twistMotion=Locked
 }
 {bone1Name=dyn_hair_fw_C_01, bone2Name=head, swing1LimitDegrees=10, swing2LimitDegrees=10, twistLimitDegrees=45, twistMotion=Locked}
 {bone1Name=dyn_hair_fw_B_01, bone2Name=head, swing1LimitDegrees=20, swing2LimitDegrees=20, twistLimitDegrees=45, twistMotion=Locked}
 {bone1Name=dyn_hair_fw_A_01, bone2Name=head, swing1LimitDegrees=10, swing2LimitDegrees=10, twistLimitDegrees=45, twistMotion=Locked}
]
