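# Post-processes the face accessory mesh of the VelvetDeskCamHelmet cosmetic character, adding rig-driven and physical secondary motion.
# Copies the attached parent's pose into a Control Rig, which receives bIsSkyDiving converted to a float each update.
# AnimDynamics then springs the forward hair bone, followed by a head-relative RigidBody simulation using a dedicated physics asset.
# An additive base_hat bone transform sits unconnected in the graph.
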AnimBlueprint /BRCosmetics/Characters/Player/Male/Medium/Bodies/M_MED_VelvetDeskCamHelmet/Meshes/Parts/M_MED_VelvetDeskCamHelmet_FaceAcc_AnimBP
target_skeleton=&/M_MED_VelvetDeskCamHelmet_FaceAcc_Skeleton
graphs [
 AnimGraph {nodes [
   "Transform (Modify) Bone - Bone: base_hat"(Alpha=1, Translation="0.372,0.011,0.381", Rotation="45.170,1.104,8.428", Scale="1,1,1") {
    rotationMode=BMM_Additive
    rotationSpace=BCS_BoneSpace
    translationMode=BMM_Additive
    translationSpace=BCS_BoneSpace
    alphaBoolBlend {blendOption=Linear}
    boneToModify {boneName=base_hat}
    rotation=<rot 45.17 1.104 8.428>
    translation=<0.372 0.011 0.381>
   }
   n1 = Control Rig!(Source=Copy Pose From Mesh() {bUseAttachedParent=true}, Alpha, isSkydiving:double=Get Skydiving() -> (Skydiving:double)) -> (Pose) {
    alphaBoolBlend {blendOption=Linear}
    controlRigAssetReference {blueprintRigClass=&/M_MED_VelvetDeskCamHelmet_FaceAcc_CtrlRig.M_MED_VelvetDeskCamHelmet_FaceAcc_CtrlRig_C}
   }
   n2 = Anim Dynamics - Bone: dyn_hair_fw_area(ComponentPose=Local To Component(LocalPose=@n1), Alpha=1, ExternalForce="0,0,0") {
    angularDampingOverride=1
    angularSpringConstant=20000
    bAngularSpring=true
    bOverrideAngularDamping=true
    bUseGravityOverride=true
    gravityScale=0.35
    simulationSpace=World
    alphaBoolBlend {blendOption=Linear}
    boundBone {boneName=dyn_hair_fw_area}
    componentLinearAccScale=<0.5 0.5 0.5>
    componentLinearVelScale=<0.5 0.5 0.5>
    gravityOverride {z=-980}
    physicsBodyDefinitions [{
      boundBone {boneName=dyn_hair_fw_area}
      boxExtents=<20 30 30>
      constraintSetup {angularLimitsMax=<10 10 10>, angularLimitsMin=<-10 -10 -10>, angularTarget=<1 0 0>}
      localJointOffset=<2 0 0>
     }]
    relativeSpaceBone {boneName=head}
   }
   n3 = RigidBody!(ComponentPose=@n2, Alpha, OverrideWorldGravity:Vector="0,0,-980", ExternalForce) -> (Pose) {
    bOverrideWorldGravity=true
    overridePhysicsAsset=&/M_MED_VelvetDeskCamHelmet_FaceAcc_Physics
    simulationSpace=BaseBoneSpace
    alphaBoolBlend {blendOption=Linear}
    baseBoneRef {boneName=head}
    componentLinearAccScale=<0.3 0.3 0.3>
    componentLinearVelScale=<0.3 0.3 0.3>
    overrideWorldGravity=<0.3 0.3 0.3>
    simSpaceSettings {maxAngularVelocity=1000, maxLinearVelocity=1000, velocityScaleZ=0.8, worldAlpha=0.7, externalAngularVelocity=<1 1 1>}
   }
   Output Pose(Result=Component To Local(ComponentPose=@n3)) {layerGroup=None}
  ]}
 EventGraph {nodes [
   Event BlueprintUpdateAnimation()
   > Set Skydiving(Skydiving:double="To Float (Boolean)"(InBool=Get bIsSkyDiving())) -> (Output_Get:double)
  ]}
]
variables [
 Detected Rot Head UP Bool {type=boolean}
 Skydiving {type=number}
]
